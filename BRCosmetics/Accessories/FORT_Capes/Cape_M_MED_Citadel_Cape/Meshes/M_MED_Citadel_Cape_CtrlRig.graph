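# Control Rig for the M_MED_Citadel_Cape cosmetic, on a full body skeleton with hair, cape, cloth and strap bones.
# Two spherical pose readers, both with debug drawing on, track clavicle_r and upperarm_l in sequence.
# Only the upperarm_l reader's output is used, weighting a Modify Transforms step on deform_glute_l.
# That step holds an identity transform, and no node drives the cape bones.
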ControlRigBlueprint /BRCosmetics/Accessories/FORT_Capes/Cape_M_MED_Citadel_Cape/Meshes/M_MED_Citadel_Cape_CtrlRig
graphs [
 RigVMFunctionLibrary {links [], nodes []}
 RigVMModel {nodes [
   RigUnit_BeginExecution = &RigUnit_BeginExecution()
   &RigUnit_SphericalPoseReader(
    ActiveRegionScaleFactors=(PositiveWidth=1,NegativeWidth=1,PositiveHeight=1,NegativeHeight=1)
    ActiveRegionSize=0.1
    Debug=(bDrawDebug=True,DebugScale=25,DebugSegments=20,DebugThickness=0.25)
    DriverAxis=(X=0,Y=1,Z=0)
    DriverItem=(Type=Bone,Name="clavicle_r")
    FalloffRegionScaleFactors=(PositiveWidth=1,NegativeWidth=1,PositiveHeight=1,NegativeHeight=1)
    FalloffSize=0.2
    OptionalParentItem=(Type=Bone)
    RotationOffset=(X=90,Y=0,Z=0)
    ExecutePin=@RigUnit_BeginExecution.ExecutePin
   )
   > n1 = &RigUnit_SphericalPoseReader(
    ActiveRegionScaleFactors=(PositiveWidth=0.5,NegativeWidth=3,PositiveHeight=1.5,NegativeHeight=1.5)
    ActiveRegionSize=0.1
    Debug=(bDrawDebug=True,DebugScale=25,DebugSegments=20,DebugThickness=0.25)
    DriverAxis=(X=1,Y=0,Z=0)
    DriverItem=(Type=Bone,Name="upperarm_l")
    FalloffRegionScaleFactors=(PositiveWidth=1,NegativeWidth=1,PositiveHeight=1,NegativeHeight=1)
    FalloffSize=0.4
    OptionalParentItem=(Type=Bone)
    RotationOffset=(X=90,Y=0,Z=0)
   )
   > &RigUnit_ModifyTransforms(
    ItemToModify=((Item=(Type=Bone,Name="deform_glute_l"),Transform=(Rotation=(X=0,Y=0,Z=0,W=1),Translation=(X=0,Y=0,Z=0),Scale3D=(X=1,Y=1,Z=1))))
    Weight=@n1.OutputParam
   )
  ]}
]
hierarchy [
 root {loc=<0 0 0>, rot=<0 0 0 1>}
 attach {parent=root, loc=<0 0 0>, rot=<0 0 0 1>}
 pelvis {parent=root, loc=<0 -0.7071 93.1191>, rot=<0.0127 0.707 -0.0127 -0.707>}
 spine_01 {parent=pelvis, loc=<6.7494 -0.6388 0>, rot=<0 0 0.034 -0.9994>}
 spine_02 {parent=spine_01, loc=<7.2518 -0.035 0>, rot=<0 0 0.0238 -0.9997>}
 spine_03 {parent=spine_02, loc=<6.0104 0 0>, rot=<0 0 0.0149 -0.9999>}
 spine_04 {parent=spine_03, loc=<8.1024 0 0>, rot=<0 0 -0.0132 -0.9999>}
 spine_05 {parent=spine_04, loc=<10.0146 0 0>, rot=<0 0 -0.0015 -1>}
 neck_01 {parent=spine_05, loc=<11.3138 0.1529 0>, rot=<0 0 -0.2003 -0.9797>}
 neck_02 {parent=neck_01, loc=<4.5627 0 0>, rot=<0 0 0.0504 -0.9987>}
 head {parent=neck_02, loc=<5.3914 0 0>, rot=<0 0 0.115 -0.9934>}
 dyn_hairMesh_r_MSTR {parent=head, loc=<2.1807 -0.4294 5.9223>, rot=<-0.0174 0.6537 0.7544 0.0573>}
 dyn_hairMesh_r_01 {parent=dyn_hairMesh_r_MSTR, loc=<0 0 0>, rot=<0 0 0 1>}
 dyn_hairMesh_r_02 {parent=dyn_hairMesh_r_01, loc=<3.9295 0.1628 0.3631>, rot=<0.0433 0.1512 -0.0582 -0.9858>}
 dyn_hairMesh_r_03 {parent=dyn_hairMesh_r_02, loc=<5.1398 0.0408 0.1336>, rot=<-0.0109 0.1353 0.0114 -0.9907>}
 dyn_hairMesh_l_MSTR {parent=head, loc=<2.0264 -0.6354 -6.3617>, rot=<0.0437 -0.6485 0.7596 0.0234>}
 dyn_hairMesh_l_01 {parent=dyn_hairMesh_l_MSTR, loc=<0 0 0>, rot=<0 0 0 1>}
 dyn_hairMesh_l_02 {parent=dyn_hairMesh_l_01, loc=<3.1606 -0.2411 -0.6106>, rot=<-0.0555 -0.1509 -0.0293 -0.9866>}
 dyn_hairMesh_l_03 {parent=dyn_hairMesh_l_02, loc=<4.8871 0.0832 -0.2355>, rot=<-0.0202 -0.1477 0.0013 -0.9888>}
 dyn_hairMeshFront_l_MSTR {parent=head, loc=<9.9217 9.4475 -1.8688>, rot=<-0.3799 -0.4323 -0.2528 -0.7778>}
 dyn_hairMeshFront_l_01 {parent=dyn_hairMeshFront_l_MSTR, loc=<0 0 0>, rot=<0 0 0 1>}
 dyn_hairMeshFront_l_02 {parent=dyn_hairMeshFront_l_01, loc=<2.1301 0 0>, rot=<0.0544 -0.7134 0.1279 -0.6868>}
 dyn_hairMeshFront_l_03 {parent=dyn_hairMeshFront_l_02, loc=<4.6674 -0.0228 -0.5571>, rot=<0.1793 -0.2269 0.0517 -0.9559>}
 dyn_hairMeshFront_l_04 {parent=dyn_hairMeshFront_l_03, loc=<5.0955 0 0>, rot=<-0.0161 -0.1436 0.0394 -0.9887>}
 dyn_hairMeshFront_l_05 {parent=dyn_hairMeshFront_l_04, loc=<5.0982 0 0>, rot=<-0.0353 -0.1926 0.0313 -0.9801>}
 dyn_hairMeshFront_l_06 {parent=dyn_hairMeshFront_l_05, loc=<3.7985 0 0>, rot=<-0.0087 -0.2247 0.0544 -0.9729>}
 dyn_hairBk_r_MSTR {parent=head, loc=<7.079 -5.1847 5.1689>, rot=<0.0039 -0.9466 0.3223 -0.0013>}
 dyn_hairBk_r_01 {parent=dyn_hairBk_r_MSTR, loc=<0 0 0>, rot=<0 0 0 1>}
 dyn_hairBk_r_02 {parent=dyn_hairBk_r_01, loc=<4.4458 0 0>, rot=<0.0595 0.0131 0.0183 -0.998>}
 dyn_hairBk_r_03 {parent=dyn_hairBk_r_02, loc=<4.4458 0 0>, rot=<0.0338 0.0066 0.2301 -0.9726>}
 dyn_hairBk_r_04 {parent=dyn_hairBk_r_03, loc=<4.4458 0 0>, rot=<-0.0054 0.0148 0.0907 -0.9958>}
 dyn_hairBk_r_05 {parent=dyn_hairBk_r_04, loc=<4.4458 0 0>, rot=<-0.0054 0.0148 0.0907 -0.9958>}
 dyn_hairBk_l_MSTR {parent=head, loc=<7.3405 -4.7699 -6.032>, rot=<0.0034 -0.8262 -0.5633 0.0023>}
 dyn_hairBk_l_01 {parent=dyn_hairBk_l_MSTR, loc=<0 0 0>, rot=<0 0 0 1>}
 dyn_hairBk_l_02 {parent=dyn_hairBk_l_01, loc=<4.5656 0 0>, rot=<0.0009 -0.0275 0.0376 -0.9989>}
 dyn_hairBk_l_03 {parent=dyn_hairBk_l_02, loc=<4.5656 0 0>, rot=<-0.017 -0.0795 0.2035 -0.9757>}
 dyn_hairBk_l_04 {parent=dyn_hairBk_l_03, loc=<4.5656 0 0>, rot=<-0.0006 -0.0259 0.0259 -0.9993>}
 dyn_hairBk_l_05 {parent=dyn_hairBk_l_04, loc=<4.5656 0 0>, rot=<-0.0006 -0.0259 0.0259 -0.9993>}
 dyn_hairBk_c_MSTR {parent=head, loc=<8.6952 -7.1551 -0.3531>, rot=<0.0041 -1 0 0>}
 dyn_hairBk_c_01 {parent=dyn_hairBk_c_MSTR, loc=<0 0 0>, rot=<0 0 0 1>}
 dyn_hairBk_c_02 {parent=dyn_hairBk_c_01, loc=<4.9571 0 0>, rot=<0 0 0.018 -0.9998>}
 dyn_hairBk_c_03 {parent=dyn_hairBk_c_02, loc=<4.9571 0 0>, rot=<0 0 0.1866 -0.9824>}
 dyn_hairBk_c_04 {parent=dyn_hairBk_c_03, loc=<4.9571 0 0>, rot=<0 0 0.1321 -0.9912>}
 dyn_hairBk_c_05 {parent=dyn_hairBk_c_04, loc=<4.9571 0 0>, rot=<0 0 -0.0669 -0.9978>}
 dyn_hairBk_c_06 {parent=dyn_hairBk_c_05, loc=<4.9571 0 0>, rot=<0 0 -0.0669 -0.9978>}
 dyn_hairTop_c_MSTR {parent=head, loc=<15.0502 1.085 0.9183>, rot=<0.0041 -1 0 0>}
 dyn_hairTop_c_01 {parent=dyn_hairTop_c_MSTR, loc=<0 0 0>, rot=<0 0 0 1>}
 attach_cape {parent=spine_05, loc=<9.017 -11.714 0>, rot=<0.0616 -0.7044 -0.0616 -0.7044>}
 clavicle_l {parent=spine_05, loc=<4.3575 -0.1466 -2.3223>, rot=<0.0465 -0.7289 -0.009 -0.6829>}
 upperarm_l {parent=clavicle_l, loc=<15.5111 0 0>, rot=<0.0226 -0.3234 -0.0088 -0.9459>}
 lowerarm_l {parent=upperarm_l, loc=<26.6514 0 0>, rot=<0.0391 -0.0099 -0.2453 -0.9686>}
 hand_l {parent=lowerarm_l, loc=<22.3491 0 0>, rot=<-0.7081 -0.0507 0.0975 -0.6975>}
 hand_attach_l {parent=hand_l, loc=<8.5 -3 0>, rot=<0.7071 0.7071 0 0>}
 thumb_01_l {parent=hand_l, loc=<1.3118 -0.3191 -2.459>, rot=<0.6546 -0.4042 0.0664 -0.6354>}
 thumb_02_l {parent=thumb_01_l, loc=<5.9336 -0.0448 0.0401>, rot=<-0.0037 0.0213 0.298 -0.9543>}
 thumb_03_l {parent=thumb_02_l, loc=<2.2223 0.3682 -0.0907>, rot=<0.0118 -0.013 0.0066 -0.9998>}
 index_metacarpal_l {parent=hand_l, loc=<1.6535 0.0183 -2.96>, rot=<0.0158 -0.0124 -0.0173 -0.9996>}
 index_01_l {parent=index_metacarpal_l, loc=<8.3008 0.0357 -0.0524>, rot=<-0.0044 0.0485 0.0924 -0.9945>}
 index_02_l {parent=index_01_l, loc=<3.3587 -0.3004 0.0452>, rot=<-0.002 0.0027 0.3225 -0.9466>}
 index_03_l {parent=index_02_l, loc=<2.5578 -0.123 -0.0252>, rot=<-0.002 0.0067 0.2558 -0.9667>}
 middle_metacarpal_l {parent=hand_l, loc=<2.1856 -0.1725 -0.5984>, rot=<0.006 0.0168 -0.0329 -0.9993>}
 middle_01_l {parent=middle_metacarpal_l, loc=<7.3275 -0.2447 0.0129>, rot=<-0.0208 0.0042 0.0858 -0.9961>}
 middle_02_l {parent=middle_01_l, loc=<4.0123 0.1401 0.1175>, rot=<-0.0218 0.0258 0.3512 -0.9357>}
 middle_03_l {parent=middle_02_l, loc=<2.6077 0.0534 0.1373>, rot=<0.0082 0.0188 0.1708 -0.9851>}
 ring_metacarpal_l {parent=hand_l, loc=<2.2074 -0.3625 1.3638>, rot=<-0.0284 0.0722 -0.0458 -0.9959>}
 ring_01_l {parent=ring_metacarpal_l, loc=<7.0295 -0.6197 -0.1446>, rot=<-0.0489 -0.0252 0.1892 -0.9804>}
 ring_02_l {parent=ring_01_l, loc=<3.6174 0.4647 -0.1027>, rot=<0 0 0.2806 -0.9598>}
 ring_03_l {parent=ring_02_l, loc=<2.4152 0.114 0.0199>, rot=<0 0 0.1189 -0.9929>}
 pinky_metacarpal_l {parent=hand_l, loc=<1.7616 -0.9753 3.3814>, rot=<-0.0397 0.0814 -0.069 -0.9935>}
 pinky_01_l {parent=pinky_metacarpal_l, loc=<7.1709 -0.4831 -0.1094>, rot=<-0.0654 -0.0035 0.2242 -0.9723>}
 pinky_02_l {parent=pinky_01_l, loc=<2.8274 0.0893 -0.1043>, rot=<0.0037 0.0011 0.2854 -0.9584>}
 pinky_03_l {parent=pinky_02_l, loc=<2.2046 0.0307 -0.0078>, rot=<-0.0002 0.0098 0.1445 -0.9895>}
 weapon_l {parent=hand_l, loc=<-0.1974 -3.4111 0.3811>, rot=<-0.0264 -0.1315 0.1356 -0.9816>}
 deform_wrist_dn_base_l {parent=hand_l, loc=<0 -2.2342 0>, rot=<0 0 0 1>}
 deform_wrist_dn_l {parent=deform_wrist_dn_base_l, loc=<0 0 0>, rot=<0 0 0 1>}
 deform_wrist_up_base_l {parent=hand_l, loc=<0 2.2682 0>, rot=<0 0 0 1>}
 deform_wrist_up_l {parent=deform_wrist_up_base_l, loc=<0 0 0>, rot=<0 0 0 1>}
 lowerarm_twist_02_l {parent=lowerarm_l, loc=<9.6199 0 0>, rot=<0 0 0 -1>}
 lowerarm_twist_01_l {parent=lowerarm_l, loc=<17.8549 0 0>, rot=<0 0 0 -1>}
 deform_elbow_in_base_l {parent=lowerarm_l, loc=<0.2112 5.259 0.8274>, rot=<0 0 0 -1>}
 deform_elbow_in_l {parent=deform_elbow_in_base_l, loc=<0 0 0>, rot=<0 0 0 1>}
 deform_elbow_out_base_l {parent=lowerarm_l, loc=<0 -4 0>, rot=<0 0 0 -1>}
 deform_elbow_out_l {parent=deform_elbow_out_base_l, loc=<0 0 0>, rot=<0 0 0 1>}
 upperarm_twist_01_l {parent=upperarm_l, loc=<0.0041 0 0>, rot=<0 0 0 -1>}
 deform_upperarm_bk_base_l {parent=upperarm_twist_01_l, loc=<1.0813 -7.2193 -1.9453>, rot=<0 0 0 -1>}
 deform_upperarm_bk_l {parent=deform_upperarm_bk_base_l, loc=<0 0 0>, rot=<0 0 0 1>}
 deform_upperarm_fr_base_l {parent=upperarm_twist_01_l, loc=<-0.3691 7.0506 -2.765>, rot=<0 0 0 -1>}
 deform_upperarm_fr_l {parent=deform_upperarm_fr_base_l, loc=<0 0 0>, rot=<0 0 0 1>}
 deform_upperarm_up_base_l {parent=upperarm_twist_01_l, loc=<-3.2216 -1.1391 3.3311>, rot=<0 0 0 -1>}
 deform_upperarm_up_l {parent=deform_upperarm_up_base_l, loc=<0 0 0>, rot=<0 0 0 1>}
 upperarm_twist_02_l {parent=upperarm_l, loc=<13.7941 0 0>, rot=<0 0 0 -1>}
 attach_backpack {parent=spine_05, loc=<-0.123 -13.473 0>, rot=<-0.0093 -0.707 0.0093 -0.707>}
 clavicle_r {parent=spine_05, loc=<4.3575 -0.1466 2.3223>, rot=<0.7289 0.0465 -0.6829 0.009>}
 upperarm_r {parent=clavicle_r, loc=<-15.5111 0 0>, rot=<0.0226 -0.3234 -0.0088 -0.9459>}
 lowerarm_r {parent=upperarm_r, loc=<-26.6514 0 0>, rot=<0.0391 -0.0099 -0.2453 -0.9686>}
 hand_r {parent=lowerarm_r, loc=<-22.3491 0 0>, rot=<-0.7081 -0.0507 0.0975 -0.6975>}
 thumb_01_r {parent=hand_r, loc=<-1.3118 0.3191 2.459>, rot=<0.6546 -0.4042 0.0664 -0.6354>}
 thumb_02_r {parent=thumb_01_r, loc=<-5.9336 0.0448 -0.0401>, rot=<-0.0037 0.0213 0.298 -0.9543>}
 thumb_03_r {parent=thumb_02_r, loc=<-2.2223 -0.3682 0.0907>, rot=<0.0118 -0.013 0.0066 -0.9998>}
 index_metacarpal_r {parent=hand_r, loc=<-1.6535 -0.0183 2.96>, rot=<0.0158 -0.0124 -0.0173 -0.9996>}
 index_01_r {parent=index_metacarpal_r, loc=<-8.3008 -0.0357 0.0524>, rot=<-0.0044 0.0485 0.0924 -0.9945>}
 index_02_r {parent=index_01_r, loc=<-3.3587 0.3004 -0.0452>, rot=<-0.002 0.0027 0.3225 -0.9466>}
 index_03_r {parent=index_02_r, loc=<-2.5578 0.123 0.0252>, rot=<-0.002 0.0067 0.2558 -0.9667>}
 middle_metacarpal_r {parent=hand_r, loc=<-2.1856 0.1725 0.5984>, rot=<0.006 0.0168 -0.0329 -0.9993>}
 middle_01_r {parent=middle_metacarpal_r, loc=<-7.3275 0.2447 -0.0129>, rot=<-0.0208 0.0042 0.0858 -0.9961>}
 middle_02_r {parent=middle_01_r, loc=<-4.0123 -0.1401 -0.1175>, rot=<-0.0218 0.0258 0.3512 -0.9357>}
 middle_03_r {parent=middle_02_r, loc=<-2.6077 -0.0534 -0.1373>, rot=<0.0082 0.0188 0.1708 -0.9851>}
 ring_metacarpal_r {parent=hand_r, loc=<-2.2074 0.3625 -1.3638>, rot=<-0.0284 0.0722 -0.0458 -0.9959>}
 ring_01_r {parent=ring_metacarpal_r, loc=<-7.0295 0.6197 0.1446>, rot=<-0.0489 -0.0252 0.1892 -0.9804>}
 ring_02_r {parent=ring_01_r, loc=<-3.6174 -0.4647 0.1027>, rot=<0 0 0.2806 -0.9598>}
 ring_03_r {parent=ring_02_r, loc=<-2.4152 -0.114 -0.0199>, rot=<0 0 0.1189 -0.9929>}
 pinky_metacarpal_r {parent=hand_r, loc=<-1.7616 0.9753 -3.3814>, rot=<-0.0397 0.0814 -0.069 -0.9935>}
 pinky_01_r {parent=pinky_metacarpal_r, loc=<-7.1709 0.4831 0.1094>, rot=<-0.0654 -0.0035 0.2242 -0.9723>}
 pinky_02_r {parent=pinky_01_r, loc=<-2.8274 -0.0893 0.1043>, rot=<0.0037 0.0011 0.2854 -0.9584>}
 pinky_03_r {parent=pinky_02_r, loc=<-2.2046 -0.0307 0.0078>, rot=<-0.0002 0.0098 0.1445 -0.9895>}
 weapon_r {parent=hand_r, loc=<0.1974 3.4111 -0.3811>, rot=<-0.0184 -0.0441 0.9967 0.0655>}
 hand_attach_r {parent=hand_r, loc=<-8.5001 2.9999 0>, rot=<0 0 -0.7071 -0.7071>}
 deform_wrist_dn_base_r {parent=hand_r, loc=<-0.0004 2.2345 -0.0001>, rot=<0 0 0 1>}
 deform_wrist_dn_r {parent=deform_wrist_dn_base_r, loc=<0 0 0>, rot=<0 0 0 1>}
 deform_wrist_up_base_r {parent=hand_r, loc=<0.0003 -2.2684 0.0001>, rot=<0 0 0 1>}
 deform_wrist_up_r {parent=deform_wrist_up_base_r, loc=<0 0 0>, rot=<0 0 0 1>}
 lowerarm_twist_02_r {parent=lowerarm_r, loc=<-9.6199 0 0>, rot=<0 0 0 -1>}
 lowerarm_twist_01_r {parent=lowerarm_r, loc=<-17.8549 0 0>, rot=<0 0 0 -1>}
 deform_elbow_in_base_r {parent=lowerarm_r, loc=<-0.2111 -5.2591 -0.8275>, rot=<0 0 0 -1>}
 deform_elbow_in_r {parent=deform_elbow_in_base_r, loc=<-0.0001 0 0.0002>, rot=<0 0 0 1>}
 deform_elbow_out_base_r {parent=lowerarm_r, loc=<0 4 -0.0001>, rot=<0 0 0 -1>}
 deform_elbow_out_r {parent=deform_elbow_out_base_r, loc=<0 0 0>, rot=<0 0 0 1>}
 upperarm_twist_01_r {parent=upperarm_r, loc=<-0.0041 0 0>, rot=<0 0 0 -1>}
 deform_upperarm_bk_base_r {parent=upperarm_twist_01_r, loc=<-1.0814 7.2192 1.9456>, rot=<0 0 0 1>}
 deform_upperarm_bk_r {parent=deform_upperarm_bk_base_r, loc=<0 0 0>, rot=<0 0 0 1>}
 deform_upperarm_fr_base_r {parent=upperarm_twist_01_r, loc=<0.3689 -7.0506 2.7652>, rot=<0 0 0 1>}
 deform_upperarm_fr_r {parent=deform_upperarm_fr_base_r, loc=<0 0 0>, rot=<0 0 0 1>}
 deform_upperarm_up_base_r {parent=upperarm_twist_01_r, loc=<3.2214 1.1391 -3.3308>, rot=<0 0 0 1>}
 deform_upperarm_up_r {parent=deform_upperarm_up_base_r, loc=<0 0 0>, rot=<0 0 0 -1>}
 upperarm_twist_02_r {parent=upperarm_r, loc=<-13.7941 0 0>, rot=<0 0 0 -1>}
 backpack_root {parent=spine_05, loc=<7.1111 -8.8171 0>, rot=<0 0 0 1>}
 hider {parent=backpack_root, loc=<0.2143 -2.5719 0>, rot=<0.0284 -0.7065 -0.0284 -0.7065>}
 cape_top {parent=hider, loc=<0 0.0288 -2.7904>, rot=<0 0 0 1>}
 dyn_cape_r_MSTR {parent=cape_top, loc=<-16.2328 1.5541 -2.5734>, rot=<-0.4358 -0.4358 0.5569 -0.5569>}
 dyn_cape_r_01 {parent=dyn_cape_r_MSTR, loc=<0.3 0 0>, rot=<0 0 0 1>}
 dyn_cape_r_02 {parent=dyn_cape_r_01, loc=<16.1032 0 0>, rot=<0 -0.0648 0 -0.9979>}
 dyn_cape_r_03 {parent=dyn_cape_r_02, loc=<16.1032 0 0>, rot=<0 -0.0149 0 -0.9999>}
 dyn_cape_r_04 {parent=dyn_cape_r_03, loc=<16.1033 0 0>, rot=<0 -0.0046 0 -1>}
 dyn_cape_r_05 {parent=dyn_cape_r_04, loc=<16.1033 0 0>, rot=<0 0 0 1>}
 dyn_cape_r_06 {parent=dyn_cape_r_05, loc=<16.1032 0 0>, rot=<0 0 0 1>}
 dyn_cape_r_07 {parent=dyn_cape_r_06, loc=<16.1031 0 0.0001>, rot=<0 0 0 1>}
 dyn_cape_l_MSTR {parent=cape_top, loc=<16.2328 1.5541 -2.5734>, rot=<-0.4358 -0.4358 0.5569 -0.5569>}
 dyn_cape_l_01 {parent=dyn_cape_l_MSTR, loc=<0.3 0 0>, rot=<0 0 0 1>}
 dyn_cape_l_02 {parent=dyn_cape_l_01, loc=<16.1032 0 0>, rot=<0 -0.0648 0 -0.9979>}
 dyn_cape_l_03 {parent=dyn_cape_l_02, loc=<16.1032 0 0>, rot=<0 -0.0149 0 -0.9999>}
 dyn_cape_l_04 {parent=dyn_cape_l_03, loc=<16.1032 0 0>, rot=<0 -0.0046 0 -1>}
 dyn_cape_l_05 {parent=dyn_cape_l_04, loc=<16.1032 0 0>, rot=<0 0 0 1>}
 dyn_cape_l_06 {parent=dyn_cape_l_05, loc=<16.1032 0 0>, rot=<0 0 0 1>}
 dyn_cape_l_07 {parent=dyn_cape_l_06, loc=<16.1032 0 0>, rot=<0 0 0 1>}
 dyn_cape_MSTR {parent=cape_top, loc=<0 -0.0317 -2.5734>, rot=<-0.4358 -0.4358 0.5569 -0.5569>}
 dyn_cape_01 {parent=dyn_cape_MSTR, loc=<0.3 0 0>, rot=<0 0 0 1>}
 dyn_cape_02 {parent=dyn_cape_01, loc=<16.1032 0 0>, rot=<0 -0.0648 0 -0.9979>}
 dyn_cape_03 {parent=dyn_cape_02, loc=<16.1032 0 0>, rot=<0 -0.0149 0 -0.9999>}
 dyn_cape_04 {parent=dyn_cape_03, loc=<16.1032 0 0>, rot=<0 -0.0046 0 -1>}
 dyn_cape_05 {parent=dyn_cape_04, loc=<16.1032 0 0>, rot=<0 0 0 1>}
 dyn_cape_06 {parent=dyn_cape_05, loc=<16.1032 0 0>, rot=<0 0 0 1>}
 dyn_cape_07 {parent=dyn_cape_06, loc=<16.1032 0 0>, rot=<0 0 0 1>}
 dyn_tentacle_1 {parent=hider, loc=<16.5246 3.996 0.4667>, rot=<0.0023 0.0011 -0.0008 -1>}
 dyn_tentacle_2 {parent=hider, loc=<-16.5701 3.983 0.4718>, rot=<0.0013 -0.0009 0.0001 -1>}
 deform_pec_aim_base_l {parent=spine_04, loc=<7.8732 12.6388 -2.4123>, rot=<0.0102 -0.4602 0.3979 -0.7936>}
 deform_pec_aim_l {parent=deform_pec_aim_base_l, loc=<0 0 0>, rot=<0 0 0 1>}
 deform_pec_base_l {parent=deform_pec_aim_l, loc=<10.766 0 0>, rot=<0 0 0 1>}
 deform_pec_l {parent=deform_pec_base_l, loc=<0.055 0 0>, rot=<0 0 0 1>}
 deform_pec_aim_base_r {parent=spine_04, loc=<7.8727 12.6388 2.4123>, rot=<-0.4602 -0.0102 0.7936 0.3979>}
 deform_pec_aim_r {parent=deform_pec_aim_base_r, loc=<0 0 0>, rot=<0 0 0 1>}
 deform_pec_base_r {parent=deform_pec_aim_r, loc=<-10.7662 -0.0006 -0.0007>, rot=<0 0 0 1>}
 deform_pec_r {parent=deform_pec_base_r, loc=<0 0 0>, rot=<0 0 0 1>}
 chest {parent=spine_04, loc=<5.5504 13.8159 0>, rot=<0.0294 -0.7065 -0.0294 -0.7065>}
 thigh_l {parent=pelvis, loc=<-2.5693 -0.0109 -7.7501>, rot=<0.0012 0.0572 0.0155 -0.9982>}
 calf_l {parent=thigh_l, loc=<-39.0484 0 0>, rot=<0.0006 -0.0161 -0.0391 -0.9991>}
 foot_l {parent=calf_l, loc=<-42.561 0.1077 0.4292>, rot=<0.0418 -0.0455 0.0343 -0.9975>}
 ball_l {parent=foot_l, loc=<-6.1599 13.8264 0.6806>, rot=<0.0001 -0.0001 -0.7063 -0.7079>}
 calf_twist_02_l {parent=calf_l, loc=<-18.8064 -0.0967 0.0397>, rot=<0.0237 0.0095 -0.0035 -0.9997>}
 calf_twist_01_l {parent=calf_l, loc=<-35.813 -0.0967 0.0397>, rot=<0.0237 0.0095 -0.0035 -0.9997>}
 deform_knee_in_base_l {parent=calf_l, loc=<1.298 -5.2957 -0.0496>, rot=<0 0 0 -1>}
 deform_knee_in_l {parent=deform_knee_in_base_l, loc=<0 0 0>, rot=<0 0 0 1>}
 deform_knee_out_base_l {parent=calf_l, loc=<0 4 0>, rot=<0 0 0 -1>}
 deform_knee_out_l {parent=deform_knee_out_base_l, loc=<0 0 0>, rot=<0 0 0 1>}
 thigh_twist_01_l {parent=thigh_l, loc=<-21.6303 0 0>, rot=<0.0013 0 0 -1>}
 deform_groin_base_l {parent=thigh_l, loc=<-2.1908 8.1142 -1.7442>, rot=<0 0 0 -1>}
 deform_groin_l {parent=deform_groin_base_l, loc=<0 0 0>, rot=<0 0 0 1>}
 dyn_strapThigh_l_MSTR {parent=thigh_l, loc=<-7.2669 0 0>, rot=<0 0 0 1>}
 thigh_r {parent=pelvis, loc=<-2.5693 -0.0109 7.7501>, rot=<0.0572 -0.0012 0.9982 0.0155>}
 calf_r {parent=thigh_r, loc=<39.0484 0 0>, rot=<0.0006 -0.0161 -0.0391 -0.9991>}
 foot_r {parent=calf_r, loc=<42.561 -0.1077 -0.4292>, rot=<0.0418 -0.0455 0.0343 -0.9975>}
 ball_r {parent=foot_r, loc=<6.1599 -13.8264 -0.6806>, rot=<0.0001 -0.0001 -0.7063 -0.7079>}
 calf_twist_02_r {parent=calf_r, loc=<18.8064 0.0967 -0.0397>, rot=<0.0237 0.0095 -0.0035 -0.9997>}
 calf_twist_01_r {parent=calf_r, loc=<35.813 0.0967 -0.0397>, rot=<0.0237 0.0095 -0.0035 -0.9997>}
 deform_knee_in_base_r {parent=calf_r, loc=<-1.298 5.2957 0.0496>, rot=<0 0 0 -1>}
 deform_knee_in_r {parent=deform_knee_in_base_r, loc=<0 0 0>, rot=<0 0 0 1>}
 deform_knee_out_base_r {parent=calf_r, loc=<0 -4 0>, rot=<0 0 0 -1>}
 deform_knee_out_r {parent=deform_knee_out_base_r, loc=<0 0 0>, rot=<0 0 0 1>}
 thigh_twist_01_r {parent=thigh_r, loc=<21.6303 0 0>, rot=<0.0013 0 0 -1>}
 deform_groin_base_r {parent=thigh_r, loc=<2.1908 -8.1142 1.7442>, rot=<0 0 0 -1>}
 deform_groin_r {parent=deform_groin_base_r, loc=<0 0 0>, rot=<0 0 0 1>}
 dyn_strapThigh_r_MSTR {parent=thigh_r, loc=<7.267 0 0>, rot=<0 0 0 1>}
 deform_glute_base_l {parent=pelvis, loc=<-8.8064 -8.8222 -8.4287>, rot=<0.0275 0.0501 -0.4472 -0.8926>}
 deform_glute_l {parent=deform_glute_base_l, loc=<0 0 0>, rot=<0 0 0 1>}
 deform_glute_base_r {parent=pelvis, loc=<-8.8062 -8.8222 8.4287>, rot=<0.0501 -0.0275 0.8926 -0.4472>}
 deform_glute_r {parent=deform_glute_base_r, loc=<0 0 0>, rot=<0 0 0 1>}
 dyn_loinClothFront_l_MSTR {parent=pelvis, loc=<5.7001 8.9472 -6.0121>, rot=<-0.0179 -0.9998 0 0>}
 dyn_loinClothFront_l_01 {parent=dyn_loinClothFront_l_MSTR, loc=<0 0 0>, rot=<0 0 0 1>}
 dyn_loinClothFront_l_02 {parent=dyn_loinClothFront_l_01, loc=<3.2284 0 0>, rot=<0 0 0 1>}
 dyn_loinClothFront_l_03 {parent=dyn_loinClothFront_l_02, loc=<8.347 0 0>, rot=<0 0 0 1>}
 dyn_loinClothFront_l_04 {parent=dyn_loinClothFront_l_03, loc=<7.9765 0 0>, rot=<0 0 0 1>}
 dyn_loinClothFront_l_05 {parent=dyn_loinClothFront_l_04, loc=<8.7758 0 0>, rot=<0 0 0 1>}
 dyn_loinClothFront_r_MSTR {parent=pelvis, loc=<5.7001 8.9472 5.9822>, rot=<-0.0179 -0.9998 0 0>}
 dyn_loinClothFront_r_01 {parent=dyn_loinClothFront_r_MSTR, loc=<0 0 0>, rot=<0 0 0 1>}
 dyn_loinClothFront_r_02 {parent=dyn_loinClothFront_r_01, loc=<3.2284 0 0>, rot=<0 0 0 1>}
 dyn_loinClothFront_r_03 {parent=dyn_loinClothFront_r_02, loc=<8.347 0 0>, rot=<0 0 0 1>}
 dyn_loinClothFront_r_04 {parent=dyn_loinClothFront_r_03, loc=<7.9765 0 0>, rot=<0 0 0 1>}
 dyn_loinClothFront_r_05 {parent=dyn_loinClothFront_r_04, loc=<8.7758 0 0>, rot=<0 0 0 1>}
 dyn_strapBack_l_MSTR {parent=pelvis, loc=<4.9889 -8.3956 -10.1344>, rot=<-0.187 -0.4254 0.8679 -0.1758>}
 dyn_strapBack_l_01 {parent=dyn_strapBack_l_MSTR, loc=<0 0 0>, rot=<0 0 0 1>}
 dyn_strapBack_l_02 {parent=dyn_strapBack_l_01, loc=<4.2638 0 0>, rot=<-0.0663 0.0051 0.0765 -0.9948>}
 dyn_strapFront_l_MSTR {parent=pelvis, loc=<4.7204 7.8335 -9.2677>, rot=<-0.1105 -0.9392 0.2038 -0.2534>}
 dyn_strapFront_l_01 {parent=dyn_strapFront_l_MSTR, loc=<0 0 0>, rot=<0 0 0 1>}
 dyn_strapFront_l_02 {parent=dyn_strapFront_l_01, loc=<4.0654 0 0>, rot=<0.0042 0.0345 0.078 -0.9963>}
 dyn_strapBack_r_MSTR {parent=pelvis, loc=<4.9889 -8.3956 10.1344>, rot=<0.187 0.4254 0.8679 -0.1758>}
 dyn_strapBack_r_01 {parent=dyn_strapBack_r_MSTR, loc=<0 0 0>, rot=<0 0 0 1>}
 dyn_strapBack_r_02 {parent=dyn_strapBack_r_01, loc=<4.2638 0 0>, rot=<0.0663 -0.0051 0.0765 -0.9948>}
 dyn_loinClothBack_l_MSTR {parent=pelvis, loc=<6.769 -9.9753 -6.3502>, rot=<0.1296 -0.9874 -0.0899 -0.0151>}
 dyn_loinClothBack_l_01 {parent=dyn_loinClothBack_l_MSTR, loc=<0 0 0>, rot=<0 0 0 1>}
 dyn_loinClothBack_l_02 {parent=dyn_loinClothBack_l_01, loc=<4.1804 0 0>, rot=<0 0 -0.0915 -0.9958>}
 dyn_loinClothBack_l_03 {parent=dyn_loinClothBack_l_02, loc=<7.9967 0 0>, rot=<0 0 -0.0787 -0.9969>}
 dyn_loinClothBack_l_04 {parent=dyn_loinClothBack_l_03, loc=<9.6613 0 0>, rot=<0 0 0 1>}
 dyn_loinClothBack_l_05 {parent=dyn_loinClothBack_l_04, loc=<8.1421 0 0>, rot=<0 0 0 1>}
 dyn_loinClothBack_r_MSTR {parent=pelvis, loc=<6.7689 -9.9753 6.3502>, rot=<0.1317 -0.985 0.1108 0.0158>}
 dyn_loinClothBack_r_01 {parent=dyn_loinClothBack_r_MSTR, loc=<0 0 0>, rot=<0 0 0 -1>}
 dyn_loinClothBack_r_02 {parent=dyn_loinClothBack_r_01, loc=<4.1803 0 0>, rot=<0 0 -0.0915 -0.9958>}
 dyn_loinClothBack_r_03 {parent=dyn_loinClothBack_r_02, loc=<7.9967 0 0>, rot=<0.0007 -0.0088 -0.0787 -0.9969>}
 dyn_loinClothBack_r_04 {parent=dyn_loinClothBack_r_03, loc=<9.6613 0 0>, rot=<0 0 0 -1>}
 dyn_loinClothBack_r_05 {parent=dyn_loinClothBack_r_04, loc=<8.1421 0 0>, rot=<0 0 0 -1>}
 dyn_strapA_l_MSTR {parent=pelvis, loc=<6.2331 4.5376 -12.9371>, rot=<-0.2116 -0.8546 0.4657 -0.0896>}
 dyn_strapA_l_01 {parent=dyn_strapA_l_MSTR, loc=<0 0 0>, rot=<0 0 0 1>}
 dyn_strapA_l_02 {parent=dyn_strapA_l_01, loc=<11.872 -1.2211 0.0231>, rot=<0 0 0 1>}
 dyn_strapC_l_MSTR {parent=pelvis, loc=<6.3282 -4.4874 -12.4931>, rot=<-0.0869 -0.5351 0.8305 -0.1283>}
 dyn_strapC_l_01 {parent=dyn_strapC_l_MSTR, loc=<0 0 0>, rot=<0 0 0 1>}
 dyn_strapC_l_02 {parent=dyn_strapC_l_01, loc=<11.9084 0 0>, rot=<-0.2333 -0.4013 0.2195 -0.8581>}
 dyn_strapB_l_MSTR {parent=pelvis, loc=<6.3099 -0.2805 -13.6832>, rot=<-0.1522 -0.7177 0.6749 -0.079>}
 dyn_strapB_l_01 {parent=dyn_strapB_l_MSTR, loc=<0 0 0>, rot=<0 0 0 1>}
 dyn_strapB_l_02 {parent=dyn_strapB_l_01, loc=<15.2294 0 0>, rot=<0 0 0 1>}
 dyn_strapA_r_MSTR {parent=pelvis, loc=<6.2331 4.5376 12.9371>, rot=<-0.2116 -0.8546 -0.4657 0.0896>}
 dyn_strapA_r_01 {parent=dyn_strapA_r_MSTR, loc=<0 0 0>, rot=<0 0 0 1>}
 dyn_strapA_r_02 {parent=dyn_strapA_r_01, loc=<11.872 -1.2212 -0.0231>, rot=<0 0 0 1>}
 dyn_strapB_r_MSTR {parent=pelvis, loc=<6.31 -0.2805 13.6832>, rot=<-0.1522 -0.7177 -0.6749 0.079>}
 dyn_strapB_r_01 {parent=dyn_strapB_r_MSTR, loc=<0 0 0>, rot=<0 0 0 1>}
 dyn_strapB_r_02 {parent=dyn_strapB_r_01, loc=<15.2294 0 0>, rot=<0 0 0 1>}
 dyn_strapC_r_MSTR {parent=pelvis, loc=<6.3282 -4.4873 12.4931>, rot=<0.0869 0.5351 0.8305 -0.1283>}
 dyn_strapC_r_01 {parent=dyn_strapC_r_MSTR, loc=<0 0 0>, rot=<0 0 0 1>}
 dyn_strapC_r_02 {parent=dyn_strapC_r_01, loc=<11.9084 0 0>, rot=<0.2333 0.4013 0.2195 -0.8581>}
 dyn_strapFront_r_MSTR {parent=pelvis, loc=<4.7204 7.8335 9.2677>, rot=<-0.1105 -0.9392 -0.2038 0.2534>}
 dyn_strapFront_r_01 {parent=dyn_strapFront_r_MSTR, loc=<0 0 0>, rot=<0 0 0 1>}
 dyn_strapFront_r_02 {parent=dyn_strapFront_r_01, loc=<4.0655 0 0>, rot=<-0.0042 -0.0345 0.078 -0.9963>}
 warp_point_03 {parent=root, loc=<0 0 0>, rot=<0 0 0 1>}
 warp_point_02 {parent=root, loc=<0 0 0>, rot=<0 0 0 1>}
 camera_root {parent=root, loc=<0 0 0>, rot=<0 0 0 1>}
 camera {parent=camera_root, loc=<0 0 0>, rot=<0 0 0 1>}
 attach_fp {parent=root, loc=<0 0 0>, rot=<0 0 0 1>}
 warp_point_01 {parent=root, loc=<0 0 0>, rot=<0 0 0 1>}
 ik_foot_root {parent=root, loc=<0 0 0>, rot=<0 0 0 1>}
 ik_foot_l {parent=ik_foot_root, loc=<15.2608 -4.4176 9.4432>, rot=<0.0357 0.7045 0.0284 -0.7082>}
 ik_foot_r {parent=ik_foot_root, loc=<-15.2608 -4.4176 9.4432>, rot=<0.7082 0.0284 -0.7045 0.0357>}
 ik_hand_root {parent=root, loc=<0 0 0>, rot=<0 0 0 1>}
 ik_hand_gun {parent=ik_hand_root, loc=<-51.9727 4.0883 102.6935>, rot=<0 0 0 1>}
 ik_hand_r {parent=ik_hand_gun, loc=<0 0 0>, rot=<0 0 0 1>}
 ik_hand_l {parent=ik_hand_gun, loc=<53.0285 82.1781 35.2051>, rot=<-0.6084 -0.1793 -0.4613 -0.6205>}
]
