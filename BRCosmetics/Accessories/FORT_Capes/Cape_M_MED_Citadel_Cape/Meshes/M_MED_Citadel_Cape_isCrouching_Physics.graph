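# Physics asset for the Citadel cape on the medium male body, in its crouching variant.
# Three seven-segment capsule chains (center, left, right) hang from master bodies on spine_05, with narrow swing limits and zero twist, widened at the top of the side chains.
# Freely jointed body and rear loincloth capsules complete the rig, with oversized leg and pelvis shapes and a huge root sphere topping out just above origin like a floor.
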PhysicsAsset /BRCosmetics/Accessories/FORT_Capes/Cape_M_MED_Citadel_Cape/Meshes/M_MED_Citadel_Cape_isCrouching_Physics
bodies {
 calf_l [{shapeType=Capsule, length=13.75, radius=17.25, shapeName=calf_l_capsule, center=<-21.3551 0 0>, rotation=<rot 0 90 90>}]
 calf_r [{shapeType=Capsule, length=13.75, radius=17.25, shapeName=calf_r_capsule, center=<21.3551 0 0>, rotation=<rot 0 90 90>}]
 dyn_cape_01 [{shapeType=Capsule, length=11.7556, radius=5.9773, shapeName=dyn_cape_01_capsule, center=<7.15 0 8>, rotation=<rot 90 0 0>}]
 dyn_cape_02 [{shapeType=Capsule, length=9.602, radius=8.1308, shapeName=dyn_cape_01_capsule, center=<7.15 0 5>, rotation=<rot 90 0 0>}]
 dyn_cape_03 [{shapeType=Capsule, length=9.602, radius=8.1308, shapeName=dyn_cape_01_capsule, center=<7.15 0 5>, rotation=<rot 90 0 0>}]
 dyn_cape_04 [{shapeType=Capsule, length=9.602, radius=8.1308, shapeName=dyn_cape_01_capsule, center=<7.15 0 5>, rotation=<rot 90 0 0>}]
 dyn_cape_05 [{shapeType=Capsule, length=14.852, radius=8.1308, shapeName=dyn_cape_01_capsule, center=<9.1492 0 5>, rotation=<rot 0 0 90>}]
 dyn_cape_06 [{shapeType=Capsule, length=14.852, radius=8.1308, shapeName=dyn_cape_01_capsule, center=<9.1492 0 5>, rotation=<rot 0 0 90>}]
 dyn_cape_07 [{shapeType=Capsule, length=14.3985, radius=8.1308, shapeName=dyn_cape_01_capsule, center=<7.8437 0 5>, rotation=<rot 0 0 90>}]
 dyn_cape_MSTR [{shapeType=Capsule, length=0.505, radius=0.505, shapeName=dyn_cape_r_MSTR_capsule, rotation=<rot 0 -90 0>}]
 dyn_cape_l_01 [{shapeType=Capsule, length=11.7556, radius=5.9773, shapeName=dyn_cape_01_capsule, center=<7.15 -2.7745 8>, rotation=<rot 90 0 0>}]
 dyn_cape_l_02 [{shapeType=Capsule, length=11.2131, radius=6.5198, shapeName=dyn_cape_01_capsule, center=<7.15 0 3>, rotation=<rot 90 0 0>}]
 dyn_cape_l_03 [{shapeType=Capsule, length=9.602, radius=8.1308, shapeName=dyn_cape_01_capsule, center=<7.15 0 5>, rotation=<rot 90 0 0>}]
 dyn_cape_l_04 [{shapeType=Capsule, length=9.602, radius=8.1308, shapeName=dyn_cape_01_capsule, center=<7.15 0 5>, rotation=<rot 90 0 0>}]
 dyn_cape_l_05 [{shapeType=Capsule, length=14.852, radius=8.1308, shapeName=dyn_cape_01_capsule, center=<9.1492 0 5>, rotation=<rot 0 0 90>}]
 dyn_cape_l_06 [{shapeType=Capsule, length=14.852, radius=8.1308, shapeName=dyn_cape_01_capsule, center=<9.1492 0 5>, rotation=<rot 0 0 90>}]
 dyn_cape_l_07 [{shapeType=Capsule, length=14.3985, radius=8.1308, shapeName=dyn_cape_01_capsule, center=<7.8437 0 5>, rotation=<rot 0 0 90>}]
 dyn_cape_l_MSTR [{shapeType=Capsule, length=0.505, radius=0.505, shapeName=dyn_cape_r_MSTR_capsule, rotation=<rot 0 -90 0>}]
 dyn_cape_r_01 [{shapeType=Capsule, length=11.7556, radius=5.9773, shapeName=dyn_cape_01_capsule, center=<7.15 0 8>, rotation=<rot 90 0 0>}]
 dyn_cape_r_02 [{shapeType=Capsule, length=11.2131, radius=6.5198, shapeName=dyn_cape_01_capsule, center=<7.15 0 3>, rotation=<rot 90 0 0>}]
 dyn_cape_r_03 [{shapeType=Capsule, length=9.602, radius=8.1308, shapeName=dyn_cape_01_capsule, center=<7.15 0 5>, rotation=<rot 90 0 0>}]
 dyn_cape_r_04 [{shapeType=Capsule, length=9.602, radius=8.1308, shapeName=dyn_cape_01_capsule, center=<7.15 0 5>, rotation=<rot 90 0 0>}]
 dyn_cape_r_05 [{shapeType=Capsule, length=14.852, radius=8.1308, shapeName=dyn_cape_01_capsule, center=<9.1492 0 5>, rotation=<rot 0 0 90>}]
 dyn_cape_r_06 [{shapeType=Capsule, length=14.852, radius=8.1308, shapeName=dyn_cape_01_capsule, center=<9.1492 0 5>, rotation=<rot 0 0 90>}]
 dyn_cape_r_07 [{shapeType=Capsule, length=14.3985, radius=8.1308, shapeName=dyn_cape_01_capsule, center=<7.8437 0 5>, rotation=<rot 0 0 90>}]
 dyn_cape_r_MSTR [{shapeType=Capsule, length=0.505, radius=0.505, shapeName=dyn_cape_r_MSTR_capsule, rotation=<rot 0 -90 0>}]
 dyn_loinClothBack_l_01 [{shapeType=Capsule, length=8, radius=6.255, shapeName=dyn_loinClothBack_l_01_capsule, center=<7.9008 0 0>, rotation=<rot 90 0 0>}]
 dyn_loinClothBack_l_03 [{shapeType=Capsule, length=8, radius=6.255, shapeName=dyn_loinClothBack_l_03_capsule, center=<7.9008 0 0>, rotation=<rot 90 0 0>}]
 dyn_loinClothBack_l_05 [{shapeType=Capsule, length=8, radius=6.255, shapeName=dyn_loinClothBack_l_05_capsule, center=<7.9008 0 0>, rotation=<rot 90 0 0>}]
 dyn_loinClothBack_r_01 [{shapeType=Capsule, length=8, radius=6.255, shapeName=dyn_loinClothBack_r_01_capsule, center=<7.9008 0 0>, rotation=<rot 90 0 0>}]
 dyn_loinClothBack_r_03 [{shapeType=Capsule, length=8, radius=6.255, shapeName=dyn_loinClothBack_r_03_capsule, center=<7.9008 0 0>, rotation=<rot 90 0 0>}]
 dyn_loinClothBack_r_05 [{shapeType=Capsule, length=8, radius=6.255, shapeName=dyn_loinClothBack_r_05_capsule, center=<7.9008 0 0>, rotation=<rot 90 0 0>}]
 pelvis [{shapeType=Capsule, length=24.5, radius=30.25, shapeName=pelvis_capsule, center=<-27.2597 13.6671 0>, rotation=<rot 0 90 90>}]
 root [{shapeType=Capsule, length=0, radius=141, shapeName=root_capsule, center=<0 -15.8151 -138.8248>, rotation=<rot 0 -90 0>}]
 spine_01 [{shapeType=Capsule, length=13.5482, radius=11.0701, shapeName=spine_01_capsule, rotation=<rot 0 -90 0>}]
 spine_02 [{shapeType=Capsule, length=13.5482, radius=11.0701, shapeName=spine_02_capsule, rotation=<rot 0 -90 0>}]
 spine_03 [{shapeType=Capsule, length=13.5482, radius=11.0701, shapeName=spine_03_capsule, rotation=<rot 0 -90 0>}]
 spine_04 [{shapeType=Capsule, length=12.6871, radius=11.9311, shapeName=spine_04_capsule, rotation=<rot 0 -90 0>}]
 spine_05 [{shapeType=Capsule, length=24.5773, radius=11.7184, shapeName=spine_05_capsule, rotation=<rot 0 -90 0>}]
 thigh_l [{shapeType=Capsule, length=22.25, radius=15, shapeName=thigh_l_capsule, center=<-17.4971 0 0>, rotation=<rot 90 0 0>}]
 thigh_r [{shapeType=Capsule, length=22.25, radius=15, shapeName=thigh_l_capsule, center=<17.4971 0 0>, rotation=<rot 90 0 180>}]
 upperarm_l [{shapeType=Capsule, length=47.5591, radius=7.9409, shapeName=upperarm_l_capsule, rotation=<rot -0.0001 -89.9999 90>}]
 upperarm_r [{shapeType=Capsule, length=73.5592, radius=7.9409, shapeName=upperarm_r_capsule, rotation=<rot -0.0001 -89.9999 90>}]
}
constraints [
 {
  bone1Name=dyn_cape_MSTR
  bone2Name=spine_05
  swing1LimitDegrees=45
  swing1Motion=Free
  swing2LimitDegrees=45
  swing2Motion=Free
  twistLimitDegrees=45
  twistMotion=Free
 }
 {bone1Name=dyn_cape_01, bone2Name=dyn_cape_MSTR, swing1LimitDegrees=5, swing2LimitDegrees=15, twistLimitDegrees=0}
 {bone1Name=dyn_cape_02, bone2Name=dyn_cape_01, swing1LimitDegrees=5, swing2LimitDegrees=15, twistLimitDegrees=0}
 {bone1Name=dyn_cape_03, bone2Name=dyn_cape_02, swing1LimitDegrees=5, swing2LimitDegrees=15, twistLimitDegrees=0}
 {bone1Name=dyn_cape_04, bone2Name=dyn_cape_03, swing1LimitDegrees=5, swing2LimitDegrees=15, twistLimitDegrees=0}
 {bone1Name=dyn_cape_05, bone2Name=dyn_cape_04, swing1LimitDegrees=5, swing2LimitDegrees=15, twistLimitDegrees=0}
 {bone1Name=dyn_cape_06, bone2Name=dyn_cape_05, swing1LimitDegrees=5, swing2LimitDegrees=15, twistLimitDegrees=0}
 {bone1Name=dyn_cape_07, bone2Name=dyn_cape_06, swing1LimitDegrees=5, swing2LimitDegrees=15, twistLimitDegrees=0}
 {
  bone1Name=dyn_cape_l_MSTR
  bone2Name=spine_05
  swing1LimitDegrees=45
  swing1Motion=Free
  swing2LimitDegrees=45
  swing2Motion=Free
  twistLimitDegrees=45
  twistMotion=Free
 }
 {bone1Name=dyn_cape_l_01, bone2Name=dyn_cape_l_MSTR, swing1LimitDegrees=5, swing2LimitDegrees=30, twistLimitDegrees=0}
 {bone1Name=dyn_cape_l_02, bone2Name=dyn_cape_l_01, swing1LimitDegrees=5, swing2LimitDegrees=30, twistLimitDegrees=0}
 {bone1Name=dyn_cape_l_03, bone2Name=dyn_cape_l_02, swing1LimitDegrees=5, swing2LimitDegrees=15, twistLimitDegrees=0}
 {bone1Name=dyn_cape_l_04, bone2Name=dyn_cape_l_03, swing1LimitDegrees=5, swing2LimitDegrees=15, twistLimitDegrees=0}
 {bone1Name=dyn_cape_l_05, bone2Name=dyn_cape_l_04, swing1LimitDegrees=5, swing2LimitDegrees=15, twistLimitDegrees=0}
 {bone1Name=dyn_cape_l_06, bone2Name=dyn_cape_l_05, swing1LimitDegrees=5, swing2LimitDegrees=15, twistLimitDegrees=0}
 {bone1Name=dyn_cape_l_07, bone2Name=dyn_cape_l_06, swing1LimitDegrees=5, swing2LimitDegrees=15, twistLimitDegrees=0}
 {
  bone1Name=dyn_cape_r_MSTR
  bone2Name=spine_05
  swing1LimitDegrees=45
  swing1Motion=Free
  swing2LimitDegrees=45
  swing2Motion=Free
  twistLimitDegrees=45
  twistMotion=Free
 }
 {bone1Name=dyn_cape_r_01, bone2Name=dyn_cape_r_MSTR, swing1LimitDegrees=5, swing2LimitDegrees=30, twistLimitDegrees=0}
 {bone1Name=dyn_cape_r_02, bone2Name=dyn_cape_r_01, swing1LimitDegrees=5, swing2LimitDegrees=30, twistLimitDegrees=0}
 {bone1Name=dyn_cape_r_03, bone2Name=dyn_cape_r_02, swing1LimitDegrees=5, swing2LimitDegrees=15, twistLimitDegrees=0}
 {bone1Name=dyn_cape_r_04, bone2Name=dyn_cape_r_03, swing1LimitDegrees=5, swing2LimitDegrees=15, twistLimitDegrees=0}
 {bone1Name=dyn_cape_r_05, bone2Name=dyn_cape_r_04, swing1LimitDegrees=5, swing2LimitDegrees=15, twistLimitDegrees=0}
 {bone1Name=dyn_cape_r_06, bone2Name=dyn_cape_r_05, swing1LimitDegrees=5, swing2LimitDegrees=15, twistLimitDegrees=0}
 {bone1Name=dyn_cape_r_07, bone2Name=dyn_cape_r_06, swing1LimitDegrees=5, swing2LimitDegrees=15, twistLimitDegrees=0}
 {
  bone1Name=calf_l
  bone2Name=thigh_l
  swing1LimitDegrees=45
  swing1Motion=Free
  swing2LimitDegrees=45
  swing2Motion=Free
  twistLimitDegrees=45
  twistMotion=Free
 }
 {
  bone1Name=calf_r
  bone2Name=thigh_r
  swing1LimitDegrees=45
  swing1Motion=Free
  swing2LimitDegrees=45
  swing2Motion=Free
  twistLimitDegrees=45
  twistMotion=Free
 }
 {
  bone1Name=pelvis
  bone2Name=root
  swing1LimitDegrees=45
  swing1Motion=Free
  swing2LimitDegrees=45
  swing2Motion=Free
  twistLimitDegrees=45
  twistMotion=Free
 }
 {
  bone1Name=upperarm_l
  bone2Name=spine_05
  swing1LimitDegrees=45
  swing1Motion=Free
  swing2LimitDegrees=45
  swing2Motion=Free
  twistLimitDegrees=45
  twistMotion=Free
 }
 {
  bone1Name=upperarm_r
  bone2Name=spine_05
  swing1LimitDegrees=45
  swing1Motion=Free
  swing2LimitDegrees=45
  swing2Motion=Free
  twistLimitDegrees=45
  twistMotion=Free
 }
 {
  bone1Name=thigh_l
  bone2Name=pelvis
  swing1LimitDegrees=45
  swing1Motion=Free
  swing2LimitDegrees=45
  swing2Motion=Free
  twistLimitDegrees=45
  twistMotion=Free
 }
 {
  bone1Name=spine_01
  bone2Name=pelvis
  swing1LimitDegrees=45
  swing1Motion=Free
  swing2LimitDegrees=45
  swing2Motion=Free
  twistLimitDegrees=45
  twistMotion=Free
 }
 {
  bone1Name=spine_02
  bone2Name=spine_01
  swing1LimitDegrees=45
  swing1Motion=Free
  swing2LimitDegrees=45
  swing2Motion=Free
  twistLimitDegrees=45
  twistMotion=Free
 }
 {
  bone1Name=spine_03
  bone2Name=spine_02
  swing1LimitDegrees=45
  swing1Motion=Free
  swing2LimitDegrees=45
  swing2Motion=Free
  twistLimitDegrees=45
  twistMotion=Free
 }
 {
  bone1Name=spine_04
  bone2Name=spine_03
  swing1LimitDegrees=45
  swing1Motion=Free
  swing2LimitDegrees=45
  swing2Motion=Free
  twistLimitDegrees=45
  twistMotion=Free
 }
 {
  bone1Name=spine_05
  bone2Name=spine_04
  swing1LimitDegrees=45
  swing1Motion=Free
  swing2LimitDegrees=45
  swing2Motion=Free
  twistLimitDegrees=45
  twistMotion=Free
 }
 {
  bone1Name=thigh_r
  bone2Name=pelvis
  swing1LimitDegrees=45
  swing1Motion=Free
  swing2LimitDegrees=45
  swing2Motion=Free
  twistLimitDegrees=45
  twistMotion=Free
 }
 {
  bone1Name=dyn_loinClothBack_l_05
  bone2Name=dyn_loinClothBack_l_03
  swing1LimitDegrees=45
  swing1Motion=Free
  swing2LimitDegrees=45
  swing2Motion=Free
  twistLimitDegrees=45
  twistMotion=Free
 }
 {
  bone1Name=dyn_loinClothBack_l_03
  bone2Name=dyn_loinClothBack_l_01
  swing1LimitDegrees=45
  swing1Motion=Free
  swing2LimitDegrees=45
  swing2Motion=Free
  twistLimitDegrees=45
  twistMotion=Free
 }
 {
  bone1Name=dyn_loinClothBack_l_01
  bone2Name=pelvis
  swing1LimitDegrees=45
  swing1Motion=Free
  swing2LimitDegrees=45
  swing2Motion=Free
  twistLimitDegrees=45
  twistMotion=Free
 }
 {
  bone1Name=dyn_loinClothBack_r_05
  bone2Name=dyn_loinClothBack_r_03
  swing1LimitDegrees=45
  swing1Motion=Free
  swing2LimitDegrees=45
  swing2Motion=Free
  twistLimitDegrees=45
  twistMotion=Free
 }
 {
  bone1Name=dyn_loinClothBack_r_03
  bone2Name=dyn_loinClothBack_r_01
  swing1LimitDegrees=45
  swing1Motion=Free
  swing2LimitDegrees=45
  swing2Motion=Free
  twistLimitDegrees=45
  twistMotion=Free
 }
 {
  bone1Name=dyn_loinClothBack_r_01
  bone2Name=pelvis
  swing1LimitDegrees=45
  swing1Motion=Free
  swing2LimitDegrees=45
  swing2Motion=Free
  twistLimitDegrees=45
  twistMotion=Free
 }
]
